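# Master head material for the Comet Winter character skin, adding an optional colored moustache to the skeletal mesh head.
# It chains the shared PreFX, CharacterSkinFX, PostFX and cinematic post functions through material attributes.
# Before PostFX, a static switch enables the moustache, tinting base and subsurface/fuzz color.
# Its mask comes from offset, center-scaled UVs sampling a fixed texture, clipped to the 0-1 range to avoid repeats.
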Material /BRCosmetics/Characters/Player/Male/Large/Bodies/M_LRG_Comet/Skins/Winter/Materials/M_LRG_Comet_Winter_Head_Master
graph {
 properties {
  bAutomaticallySetUsageInEditor=false
  bUseMaterialAttributes=true
  bUsedWithMeshDeformer=true
  bUsedWithMorphTargets=true
  bUsedWithSkeletalMesh=true
  pixelDepthOffsetMode=PDOM_Legacy
  refractionMethod=RM_IndexOfRefraction
  shadingModel=MSM_FromMaterialExpression
  subsurfaceProfile=&SS_HeroSkin_02
  usedShadingModels=Default Lit | Subsurface | Clear Coat | Subsurface Profile | Cloth | Eye
 }
 nodes [
  n1 = MaterialFunctionCall() {materialFunction=&BaseFN_CharacterFunction_PreFX}
  n2 = MaterialFunctionCall(
   PreskinnedPosition=@n1.Pre skinned Local Position
   Gmap=@n1.Pre skinned Local Position
   MaterialAttributes=@n1.Pre skinned Local Position
   AO=@n1.Pre skinned Local Position
  ) {materialFunction=&CharacterSkinFX}
  n3 = GetMaterialAttributes(None=@n2.Result) {attributeGetTypes [#1]}
  n4 = TextureCoordinate()
  n5 = AppendVector(
   A=ScalarParameter() {defaultValue=-0.21, group=Winter Moustache, parameterName=Moustache X Position}
   B=ScalarParameter() {defaultValue=-0.2, group=Winter Moustache, parameterName=Moustache Y Position}
  )
  n6 = TextureObjectParameter() {
   group=Winter Moustache
   parameterName=Mask
   samplerType=SAMPLERTYPE_Alpha
   texture=/BRCosmetics/Characters/Player/Male/Large/Bodies/M_LRG_Comet/Skins/Winter/Textures/T_M_LRG_Comet_Winter_Moustache_temp
  }
  n7 = ScalarParameter() {defaultValue=0.055, group=Winter Moustache, parameterName=Moustache Scale}
  n8 = ScalarParameter() {defaultValue=1, group=Winter Moustache, parameterName=Moustache Scale V}
  n9 = MaterialFunctionCall(
   UVs=MaterialFunctionCall!(UVs=Add(A=@n4, B=@n5), TextureIn=@n6, UpperLeftCorner, LowerRightCorner, ExportFloat4) {materialFunction=/Engine/Functions/Engine_MaterialFunctions01/Texturing/TextureCropping}.Crop UVs
   Texture Scale=AppendVector(A=@n7, B=Multiply(A=@n7, B=@n8))
  ) {materialFunction=&ScaleUVsByCenter}
  n10 = TextureSample(UVs=@n9.UVs) {
   samplerSource=SSM_Clamp_WorldGroupSettings
   samplerType=SAMPLERTYPE_Masks
   texture=&T_M_LRG_Comet_Winter_Moustache_temp
  }
  n11 = Saturate(None=Multiply(A=@n10.R, B=Ceil(None=@n9.0-1 mask)))
  n12 = NamedRerouteDeclaration(None=@n11) {name=outMask, nodeColor=<rgba 0.1719 1 0 1>}
  n13 = LinearInterpolate(
   A=@n3.Base Color
   B=VectorParameter() {group=Winter Moustache, parameterName=Moustache Color, defaultValue=<rgba 1 1 1 1>}.RGB
   Alpha=Saturate(None=Multiply(A=@n12) {constB=2})
  ) {constAlpha=1}
  n14 = MaterialFunctionCall(
   AO=@n1.AO
   Distance Blend=@n1.AO
   Fresnel=@n1.AO
   Gmap=@n1.AO
   Pre Skinned Local Position=@n1.AO
   Material=StaticSwitchParameter(
    True=SetMaterialAttributes!(MaterialAttributes=@n3.MaterialAttributes, Base Color=@n13, Fuzz Color or Subsurface Color=@n13) {attributeSetTypes [#1, #2]}
    False=@n2.Result
   ) {group=Winter Moustache, parameterName=useWinterMoustache}
  ) {materialFunction=&BaseFN_CharacterFunction_PostFX}
  MaterialFunctionCall(In=@n14.Result) {materialFunction=&CINE_MF_CharPostFX}
  NamedRerouteUsage()
  Reroute(None=@n1.Result)
  Subtract(A=@n11, B=ScalarParameter() {defaultValue=0.5, group=Winter Moustache, parameterName=MaskBias}) {constB=0.5}
 ]
}
